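# Niagara system for a snow chunk burst effect, built from a single sprite emitter named "Empty" with no user parameters.
# An instantaneous burst spawns sprites in a small sphere with random lifetime and size, launched by two AddVelocity modules.
# Gravity, randomized drag and rotation shape their motion, while curves shrink and fade them over life.
# Speed-based scaling further sizes the sprites, and material random is written for per-particle variation.
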NiagaraSystem /Game/NeverCook/FX/Niagara_FX_Reuse_Examples/Snow/NS_Snow_Chunk_GPU
dependencies {
 usedDataInterfaces [&NiagaraDataInterfaceCurve, &NiagaraDataInterfaceVector2DCurve]
 usedDynamicInputs []
 usedModules [&SystemState]
}
emitters {Empty {
  emitterUpdateScript [
   &EmitterState(Life Cycle Mode=::NewEnumerator1, Loop Behavior=::NewEnumerator1, Loop Duration=1, UseLoopDelay=false)
   &SpawnBurst_Instantaneous(Spawn Count=30, Use Spawn Probability=false, UseLoopCountLimit=false)
  ]
  particleSpawnScript [
   &InitializeParticle(
    Color Mode=::NewEnumerator0
    Lifetime Max=2.5
    Lifetime Min=0.25
    Lifetime Mode=::NewEnumerator1
    Sprite Size Max=<10 18>
    Sprite Size Min=<5 8>
    Sprite Size Mode=::NewEnumerator2
    UsePositionOffset=false
    Write Material Random=true
   )
   &ShapeLocation(MaskSpawnGroup=false, Override Seed=false, Sphere Radius=10, UseOwnerScale=true)
   &AddVelocity(Module.Use Falloff Distance=false, Override Seed=false, Velocity Mode=::NewEnumerator1, Velocity Speed=50) {module=AddVelocity001}
   &AddVelocity()
  ]
  particleUpdateScript [
   &ParticleState()
   &SpriteRotationRate(Rotation Rate=&RandomRangeFloat(Minimum=-0.5, Maximum=0.5))
   &ScaleSpriteSize(
    Non-Uniform Curve Sprite Scale=DataInterface() {propertyValues {
      XCurve {keys [<at 0 1 RCIM_Cubic>, {arriveTangent=-1.092, interpMode=RCIM_Cubic, leaveTangent=-1.092, tangentMode=RCTM_User, time=1, value=0}]}
      YCurve {keys [{arriveTangent=-4.3893, interpMode=RCIM_Cubic, leaveTangent=-4.3893, tangentMode=RCTM_User, time=0, value=2}, <at 1 0.1 RCIM_Cubic>]}
     }}
    Scale Sprite Size Mode=::NewEnumerator3
   )
   &ScaleColor(
    Scale Alpha=&FloatFromCurve(FloatCurve=DataInterface() {propertyValues {Curve {keys [
         {arriveTangent=4.7492, interpMode=RCIM_Cubic, leaveTangent=4.7492, tangentMode=RCTM_User, time=0, value=0.6}
         {arriveTangent=-1.2847, interpMode=RCIM_Cubic, leaveTangent=-1.2847, tangentMode=RCTM_User, time=0.1, value=1}
         <at 1 0 RCIM_Cubic>
        ]}}})
    Scale RGB=&MakeVectorFromLinearColorRGB(Color=<rgba 0.4844 0.4844 0.4844 1>)
    ScaleA=true
    ScaleRGB=true
   )
   &GravityForce(Gravity=<0 0 -300>)
   &Drag(Drag=&RandomRangeFloat(Minimum=0.25, Maximum=1.5), Use Linear Drag=true, Use Rotational Drag=true)
   &SolveForcesAndVelocity(Clamp Velocity=false, Limit Acceleration=false)
   &ScaleSpriteSizeBySpeed(Max Scale Factor=<1 2>, Min Scale Factor=<0.1 0.01>)
  ]
  rendererClasses [&NiagaraSpriteRendererProperties]
  renderers [{rendererClass=&NiagaraSpriteRendererProperties}]
 }}
summary {
 systemName=NS_Snow_Chunk_GPU
 emitters [{emitterName=Empty, rendererClasses [&NiagaraSpriteRendererProperties]}]
 userVariables []
}
